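; Holiday llama ornament prop Blueprint with configurable emissive glow.
; Its construction script creates a dynamic material instance from M_Holiday_LlamaOrnament_A on slot 0 of the static mesh component.
; It then feeds the ornament config's emissive color into the Emissive vector parameter and its emissive intensity into the Brightness scalar parameter.
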(fn ConstructionScript
  (bind _returnvalue (Rendering|Material|CreateDynamicMaterialInstance (Variables|StaticMeshActor|GetStaticMeshComponent) 0 "/Game/Environments/Asteria/Props/Holiday/Holiday_LlamaOrnament_A/Materials/M_Holiday_LlamaOrnament_A.M_Holiday_LlamaOrnament_A" "None" :!class /Script/Engine.PrimitiveComponent))
  (Rendering|Material|SetVectorParameterValue _returnvalue "Emissive" (Variables|OrnamentConfig|GetEmissiveColor) :!class /Script/Engine.MaterialInstanceDynamic)
  (Rendering|Material|SetScalarParameterValue _returnvalue "Brightness" (Variables|OrnamentConfig|GetEmissiveIntensity) :!class /Script/Engine.MaterialInstanceDynamic))
