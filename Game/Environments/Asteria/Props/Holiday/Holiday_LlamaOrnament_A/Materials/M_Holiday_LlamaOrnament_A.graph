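# Nanite-enabled Default Lit material for the holiday llama ornament prop, with texture-driven surface and tinted emissive glow, animated to twist back and forth.
# World Position Offset rotates vertices around vertical Z at the object pivot by a time-driven sine, weakened with height by a bounding-box mask.
# Pivot location varies each object's phase and speed.
# World-space normals sum the original and rotated normal map directions.
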Material /Game/Environments/Asteria/Props/Holiday/Holiday_LlamaOrnament_A/Materials/M_Holiday_LlamaOrnament_A
graph {
 properties {
  bTangentSpaceNormal=false
  bUseMaterialAttributes=true
  bUsedWithNanite=true
  pixelDepthOffsetMode=PDOM_Legacy
  previewMesh=/Game/Environments/Asteria/Props/Holiday/Holiday_LlamaOrnament_A/Meshes/SM_Holiday_LlamaOrnament_A
  usedShadingModels=Default Lit
 }
 nodes [
  n1 = MaterialFunctionCall() {materialFunction=&BoundingBoxBased_0-1_UVW}
  n2 = ScalarParameter() {defaultValue=0.25, parameterName=TwistMaskPercent, sliderMax=1}
  n3 = Saturate(None=Divide(A=Subtract(A=@n1.B, B=@n2) {constB=0.2}, B=Add(A=@n2)))
  Ceil(None=@n3) {desc=Use to preview mesh mask}
  n4 = MaterialFunctionCall() {materialFunction=&ObjectPivotPoint}
  n5 = Frac(None=Divide(A=MaterialFunctionCall(f3=@n4.Object Pivot Location) {materialFunction=&AddComponents}.f3) {constB=100})
  Multiply(A=ScalarParameter() {defaultValue=0.06, parameterName=RotateSpeed}, B=@n5)
  ScalarParameter() {defaultValue=1, parameterName=Metallic, sliderMax=1}
  n6 = TextureSample() {samplerType=SAMPLERTYPE_Masks, texture=/Game/Environments/Asteria/Props/Holiday/Holiday_LlamaOrnament_A/Textures/T_LlamaOrnament__A_A_S}
  n7 = TextureSample() {samplerType=SAMPLERTYPE_Normal, texture=/Game/Environments/Asteria/Props/Holiday/Holiday_LlamaOrnament_A/Textures/T_LlamaOrnament__A_A_N}
  n8 = TextureSample() {samplerType=SAMPLERTYPE_Masks, texture=/Game/Environments/Asteria/Props/Holiday/Holiday_LlamaOrnament_A/Textures/T_LlamaOrnament__A_A_E}
  n9 = VectorParameter() {parameterName=Emissive, defaultValue=<rgba 1 0 0 1>}
  n10 = SetMaterialAttributes!(
   MaterialAttributes
   Base Color=Multiply(
    A=TextureSample() {texture=/Game/Environments/Asteria/Props/Holiday/Holiday_LlamaOrnament_A/Textures/T_LlamaOrnament__A_A_D}.RGB
    B=ScalarParameter() {defaultValue=1.5, parameterName=BaseColor Brightness}
   )
   Metallic=@n6.G
   Roughness=Add(A=@n6.B, B=ScalarParameter() {defaultValue=0.156, parameterName=RoughnessBias, sliderMax=1, sliderMin=-1})
   Normal=@n7.RGB
   Emissive Color=Multiply(A=Multiply(A=@n8.R, B=@n9.RGB), B=ScalarParameter() {defaultValue=8, parameterName=Brightness})
   Specular=@n6.G
  ) {attributeSetTypes [
    #1
    #2
    #3
    #4
    #5
    #6
   ]}
  n11 = Constant3Vector() {constant=<rgba 0 0 1 0>}
  n12 = OneMinus(None=@n3)
  n13 = ScalarParameter() {defaultValue=0.09, parameterName=RotateAmt}
  n14 = Multiply(A=Add(A=Time(), B=@n5), B=LinearInterpolate(Alpha=@n5) {constA=0.1, constB=0.3})
  n15 = NamedRerouteDeclaration(None=Multiply(A=Multiply(A=@n12, B=@n13), B=Sine(None=@n14))) {name=RotationAngle, nodeColor=<rgba 1 0.5859 0 1>}
  n16 = Transform(None=@n7.RGB)
  n17 = RotateAboutAxis(NormalizedRotationAxis=@n11, RotationAngle=@n15, PivotPoint=Constant3Vector(), Position=@n16)
  SetMaterialAttributes!(
   MaterialAttributes=@n10
   World Position Offset=RotateAboutAxis(
    NormalizedRotationAxis=@n11
    RotationAngle=@n15
    PivotPoint=MaterialFunctionCall() {materialFunction=&ObjectPivotPoint}.Object Pivot Location
    Position=WorldPosition().XYZ
   )
   Normal=NamedRerouteDeclaration(None=Add(A=@n16, B=@n17)) {name=RotatedNormal, nodeColor=<rgba 0 1 0.5312 1>}
  ) {attributeSetTypes [#7, #4]}
 ]
}
